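# UI material for the Masamune charge meter, drawing a translucent arc-shaped progress bar.
# A circle outline and rotated single-segment radial mask shape the arc.
# AnimateProgress interpolates between ProgressStart and ProgressEnd, driving a linear wipe along rotated UVs rather than an angular sweep.
# The wipe splits filled and empty parts with separate opacity control, and Switch blends ProgressColor toward white.
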Material /MasamuneGameplay/Assets/UI/Materials/M_UI_MasamuneCharge
graph {
 properties {bAutomaticallySetUsageInEditor=false, blendMode=BLEND_Translucent, materialDomain=MD_UI, usedShadingModels=Default Lit}
 nodes [
  LinearInterpolate(
   A=LinearInterpolate(
    A=VectorParameter() {parameterName=ProgressColor, defaultValue=<rgba 0.0395 1 0.2086 1>}.RGB
    B=Constant3Vector() {constant=<rgba 1 1 1 0>}
    Alpha=ScalarParameter() {parameterName=Switch}
   )
   B=Constant3Vector()
   Alpha=NamedRerouteUsage()
  )
  n1 = NamedRerouteUsage()
  n2 = NamedRerouteUsage()
  LinearInterpolate(
   A=LinearInterpolate(A=NamedRerouteUsage(), B=@n1, Alpha=@n1)
   B=Multiply(A=@n2, B=ScalarParameter() {defaultValue=0.5, parameterName=EmptyPartOpacity})
   Alpha=@n2
  )
  n3 = Constant() {r=0.85}
  n4 = Constant() {r=0.08}
  NamedRerouteDeclaration(None=Multiply(
    A=MaterialFunctionCall(Size=@n3, OutlineThickness=@n4) {materialFunction=&MF_UI_Circle}.Outline
    B=NamedRerouteUsage()
   )) {name=ProgressBarShape, nodeColor=<rgba 0 0 0 1>}
  n5 = MaterialFunctionCall(UVs=TextureCoordinate(), "Rotation Angle (0-1)"=Constant() {r=-0.25}) {materialFunction=&CustomRotator}
  n6 = Constant() {r=1}
  n7 = Constant() {r=0.704}
  NamedRerouteDeclaration(None=SmoothStep(Value=MaterialFunctionCall!(UV=@n5.Rotated Values, Segment Count=@n6, Segment Spacing=@n7, Fixed Spacing) {materialFunction=/FortUILibrary/MaterialFunctions/Basics/MF_UI_RadialSegments}.Segmented Triangles Gradient) {constMax=0.01}) {name=Mask, nodeColor=<rgba 0 0 0 1>}
  n8 = NamedRerouteUsage()
  n9 = ScalarParameter() {defaultValue=1, parameterName=ProgressOpacity}
  NamedRerouteDeclaration(None=Multiply(A=NamedRerouteUsage(), B=Multiply(A=@n8, B=@n9))) {name=ProgressBar, nodeColor=<rgba 0 0 0 1>}
  n10 = OneMinus(None=NamedRerouteUsage())
  n11 = ScalarParameter() {defaultValue=1, parameterName=ProgressOpacity}
  NamedRerouteDeclaration(None=Multiply(A=NamedRerouteUsage(), B=Multiply(A=@n10, B=@n11))) {name="ProgressBar(1-x)", nodeColor=<rgba 0 0 0 1>}
  n12 = ScalarParameter() {group=Progress_Layout, parameterName=ProgressEdgeSharpness, sliderMax=1, sortPriority=2}
  n13 = MaterialFunctionCall(UVs=TextureCoordinate(), "Rotation Angle (0-1)"=Constant() {r=-0.5}) {materialFunction=&CustomRotator}
  n14 = ScalarParameter() {
   defaultValue=0.1
   desc="When progress value changes, it defines where change starts.\r\n0 - 1"
   group=Progress
   parameterName=ProgressStart
   sliderMax=1
   sortPriority=1
  }
  n15 = ScalarParameter() {
   defaultValue=0.9
   desc="When progress value changes, it defines where change ends.\r\n0 - 1"
   group=Progress
   parameterName=ProgressEnd
   sliderMax=1
   sortPriority=2
  }
  n16 = ScalarParameter() {
   desc="Property for animating progress value change (from ProgressStart to ProgressEnd)\r\n0 - 1"
   group=Progress
   parameterName=AnimateProgress
   sliderMax=1
   sortPriority=3
  }
  n17 = LinearInterpolate(A=@n14, B=@n15, Alpha=@n16)
  n18 = ComponentMask(None=Add(A=Multiply(A=@n13.Rotated Values) {constB=-1}, B=NamedRerouteDeclaration(None=@n17) {name=Progress, nodeColor=<rgba 0 0 0 1>}) {constB=0.5}) {g=true}
  NamedRerouteDeclaration(None=Saturate(None=SmoothStep(Max=@n12, Value=@n18) {constMax=0.01})) {name=ProgressGraident, nodeColor=<rgba 0 0 0 1>}
 ]
}
