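# Translucent UI material that draws the Masamune reticle: mirrored brackets with a progress fill, plus a shadowed center dot.
# Brackets come from a channel-packed distance-field texture mirrored through a symmetry function, with adjustable offset and shadow that ShowMaskFx blends toward panning noise.
# AnimateProgress thresholds a gradient mask, clipped by named-reroute masks, into a fill with a glowing leading edge.
# Named reroutes feed these masks into summed color and opacity, using gold gradients and GlowColor.
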Material /MasamuneGameplay/Assets/UI/Materials/M_UI_Masamune_Reticle
graph {
 properties {bAutomaticallySetUsageInEditor=false, blendMode=BLEND_Translucent, materialDomain=MD_UI, usedShadingModels=Default Lit}
 nodes [
  n1 = Constant3Vector() {constant=<rgba 1 0.8549 0.307 1>}
  n2 = Constant3Vector() {constant=<rgba 0.9647 0.8936 0.6805 1>}
  n3 = ComponentMask(None=TextureCoordinate()) {g=true}
  n4 = Constant() {r=0.53}
  n5 = Saturate(None=Multiply(A=Subtract(A=@n3, B=@n4)) {constB=9})
  n6 = LinearInterpolate(
   A=Constant3Vector()
   B=LinearInterpolate(A=LinearInterpolate(A=@n1, B=@n2, Alpha=@n5), B=Constant3Vector() {constant=<rgba 1 0.8834 0.1615 1>}, Alpha=NamedRerouteUsage())
   Alpha=NamedRerouteUsage()
  )
  n7 = NamedRerouteUsage()
  n8 = VectorParameter() {parameterName=GlowColor, defaultValue=<rgba 1 0.4977 0 1>}
  n9 = Multiply(A=NamedRerouteUsage(), B=NamedRerouteUsage())
  Add(
   A=Add(A=Multiply(A=@n6, B=@n7), B=Multiply(A=@n8.RGB, B=@n9))
   B=Multiply(A=NamedRerouteUsage(), B=Constant3Vector() {constant=<rgba 1 1 1 0>})
  )
  n10 = NamedRerouteUsage()
  n11 = ScalarParameter() {defaultValue=0.6, parameterName=BracketEmptyOpacity}
  n12 = Max(A=@n10, B=NamedRerouteUsage())
  n13 = NamedRerouteUsage()
  Add(A=Add(A=Multiply(A=@n10, B=@n11) {constB=0.6}, B=Multiply(A=@n12, B=@n13)), B=Add(A=NamedRerouteUsage(), B=NamedRerouteUsage()))
  n14 = NamedRerouteUsage()
  n15 = ComponentMask(None=@n14) {r=true}
  n16 = Subtract(A=ComponentMask(None=@n14) {g=true}, B=Constant() {r=0.078})
  n17 = Add(A=Constant() {r=0.095}, B=ScalarParameter() {parameterName=BracketsOffset})
  n18 = Constant()
  n19 = Constant2Vector() {g=0.3, r=0.6}
  n20 = TextureSample(UVs=MaterialFunctionCall!(UV=AppendVector(A=@n15, B=@n16), Offset=AppendVector(A=@n17, B=@n18), Rotation, Scale=Multiply(A=@n19) {constB=0.8}, MirrorY) {materialFunction=/FortUILibrary/MaterialFunctions/Basics/MF_UI_Symmetry_2x1_DEPRECATED}.Result) {
   samplerSource=SSM_Clamp_WorldGroupSettings
   samplerType=SAMPLERTYPE_LinearGrayscale
   texture=/MasamuneGameplay/Assets/UI/Textures/T_UI_Masamune_Reticle
  }
  n21 = SmoothStep(Value=@n20.R) {constMin=0.1}
  n22 = Multiply(A=@n21) {constB=0.3}
  n23 = MaterialFunctionCall(UV=TextureCoordinate(), Scale=Constant2Vector() {g=0.2, r=0.4}) {materialFunction=&MF_UI_Symmetry_2x1_DEPRECATED}
  n24 = Multiply(
   A=@n21
   B=TextureSample(UVs=Panner(Coordinate=@n23.Result) {speedX=0.1, speedY=0.2}) {texture=/MasamuneGameplay/Assets/UI/Textures/T_UI_Masamune_Noise}.RGB
  )
  n25 = Multiply(
   A=LinearInterpolate(A=@n21, B=Add(A=@n22, B=@n24), Alpha=ScalarParameter() {parameterName=ShowMaskFx})
   B=ScalarParameter() {defaultValue=1, parameterName=BracketShadowOpacity}
  ) {constB=0.6}
  n26 = Constant() {r=0.5}
  n27 = Constant() {r=0.9}
  n28 = NamedRerouteDeclaration(None=ComponentMask(None=MaterialFunctionCall!(RGBA Channels=@n20.RGBA, Glow Min=@n26, Glow Max=@n27, Outline Thickness) {materialFunction=/FortUILibrary/MaterialFunctions/Basics/MF_UI_ChannelPackedDistanceField}.Fill) {b=true, g=true, r=true}) {name=Brackets, nodeColor=<rgba 0 0 0 1>}
  NamedRerouteDeclaration(None=Saturate(None=Subtract(A=@n25, B=@n28))) {name=BracketsGlow, nodeColor=<rgba 0 0 0 1>}
  NamedRerouteDeclaration(None=TextureCoordinate()) {name=UV, nodeColor=<rgba 0 0 0 1>}
  n29 = MaterialFunctionCall(UVs=NamedRerouteUsage(), Size=Constant() {r=0.02}) {materialFunction=&MF_UI_Circle}
  NamedRerouteDeclaration(None=OneMinus(None=SmoothStep(Value=@n29.SDF) {constMax=0.02, constMin=0.005})) {name=Dot, nodeColor=<rgba 0 0 0 1>}
  n30 = SmoothStep(Value=@n29.SDF) {constMax=0.04, constMin=0.006}
  NamedRerouteDeclaration(None=Multiply(A=OneMinus(None=@n30)) {constB=0.6}) {name=DotShadow, nodeColor=<rgba 0 0 0 1>}
  n31 = ScalarParameter() {
   defaultValue=1
   desc="Property for animating progress value change (from ProgressStart to ProgressEnd)\r\n0 - 1"
   group=Progress
   parameterName=AnimateProgress
   sliderMax=1
   sortPriority=3
  }
  n32 = OneMinus(None=@n31)
  n33 = Min(
   A=Subtract(
    A=TextureSample() {samplerSource=SSM_Clamp_WorldGroupSettings, texture=/MasamuneGameplay/Assets/UI/Textures/T_UI_Masamune_GradientMask}.R
    B=LinearInterpolate(Alpha=@n32)
   )
   B=Max(A=NamedRerouteUsage(), B=NamedRerouteUsage())
  )
  n34 = Ceil(None=@n33)
  NamedRerouteDeclaration(None=@n34) {name=ProgressBar, nodeColor=<rgba 0 0 0 1>}
  n35 = Multiply(A=@n34, B=Subtract(A=OneMinus(None=@n33)) {constB=0.9})
  NamedRerouteDeclaration(None=Saturate(None=Multiply(A=@n35) {constB=15})) {name=ProgressGlow, nodeColor=<rgba 0 0 0 1>}
 ]
}
