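# Weapon reticle extension widget for Masamune that shows its cooldown through animated parameters on the reticle material.
# On construct, it watches the owning controller for Granted.Cooldown.Masamune and reads that effect's remaining and total duration.
# HanldeCooldown plays one-second Anim_CoolDown_02, offset and speed-scaled to match the cooldown.
# A tag-removed listener with no tag set resets AnimateProgress to 1; other animations and the hidden charge image go unused.
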WidgetBlueprint /MasamuneGameplay/Assets/UI/WBP_MasamuneReticle
animations [
 Anim_CoolDown {
  binding_count=1
  length_seconds=1
  object=&self:Anim_CoolDown
  sequence {bindings [Reticle {
     id=#1
     possessed_class=FortSZAwareImage
     tracks [{
       type=MovieSceneWidgetMaterialTrack
       sections [{
         type=MovieSceneParameterSection
         end_frame=20
         start_frame=0
         channels [
          AnimateProgress {type=MovieSceneScriptingFloatChannel, keys [[0,0],[19.0187,1]]}
          BracketShadowOpacity {type=MovieSceneScriptingFloatChannel, keys [[0,1],[18.7557,1],[19.402,2],[20,1]]}
          GlowColorR {type=MovieSceneScriptingFloatChannel, const [1,4]}
          GlowColorG {type=MovieSceneScriptingFloatChannel, keys [[0,0.4977],[18.7557,0.4977],[19.402,1],[20,0.4977]]}
          GlowColorB {type=MovieSceneScriptingFloatChannel, keys [[0,0],[18.7557,0],[19.402,1],[20,0]]}
          GlowColorA {type=MovieSceneScriptingFloatChannel, const [1,4]}
         ]
        }]
      }]
    }]}
 }
 Anim_Charged {
  binding_count=1
  length_seconds=0.2
  object=&self:Anim_Charged
  sequence {bindings [Reticle {
     id=#1
     possessed_class=FortSZAwareImage
     tracks [{
       type=MovieSceneWidgetMaterialTrack
       sections [{
         type=MovieSceneParameterSection
         end_frame=4
         start_frame=0
         channels [
          AnimateProgress {type=MovieSceneScriptingFloatChannel, const [1,1]}
          BracketShadowOpacity {type=MovieSceneScriptingFloatChannel, keys [[0,1],[2,2],[4,1]]}
          GlowColorR {type=MovieSceneScriptingFloatChannel, const [1,3]}
          GlowColorG {type=MovieSceneScriptingFloatChannel, keys [[0,0.4977],[2,1],[4,0.4977]]}
          GlowColorB {type=MovieSceneScriptingFloatChannel, keys [[0,0],[2,1],[4,0]]}
          GlowColorA {type=MovieSceneScriptingFloatChannel, const [1,3]}
         ]
        }]
      }]
    }]}
 }
 Anim_ChargedAttack {
  binding_count=1
  length_seconds=1.2
  object=&self:Anim_ChargedAttack
  sequence {bindings [Reticle {
     id=#2
     possessed_class=FortSZAwareImage
     tracks [{
       type=MovieSceneWidgetMaterialTrack
       sections [{
         type=MovieSceneParameterSection
         end_frame=24
         start_frame=0
         channels [
          BracketsOffset {type=MovieSceneScriptingFloatChannel, keys [[0,0],[14,0],[17,0.16],[20,0]]}
          BracketShadowOpacity {type=MovieSceneScriptingFloatChannel, keys [[0,1],[14,3],[20,1]]}
          ShowMaskFx {type=MovieSceneScriptingFloatChannel, keys [[0,0],[14,1],[17,1],[20,0]]}
          AnimateProgress {type=MovieSceneScriptingFloatChannel, keys [[0,1],[20,1],[24,0]]}
          GlowColorR {type=MovieSceneScriptingFloatChannel, const [1,1]}
          GlowColorG {type=MovieSceneScriptingFloatChannel, const [0.4977,1]}
          GlowColorB {type=MovieSceneScriptingFloatChannel, const [0,1]}
          GlowColorA {type=MovieSceneScriptingFloatChannel, const [1,1]}
         ]
        }]
      }]
    }]}
 }
 Anim_CoolDown_02 {
  binding_count=1
  length_seconds=1
  object=&self:Anim_CoolDown_02
  sequence {bindings [Reticle {
     id=#1
     possessed_class=FortSZAwareImage
     tracks [{
       type=MovieSceneWidgetMaterialTrack
       sections [{
         type=MovieSceneParameterSection
         end_frame=20
         start_frame=0
         channels [
          AnimateProgress {type=MovieSceneScriptingFloatChannel, keys [[0,1],[0.794,1],[1.193,0],[19.0187,1]]}
          BracketShadowOpacity {type=MovieSceneScriptingFloatChannel, keys [[0,1],[0.1227,3],[0.671,3],[1.1927,1],[18.7557,1],[19.402,2],[20,1]]}
          BracketsOffset {type=MovieSceneScriptingFloatChannel, keys [[0,0],[0.394,0.16],[0.7943,0]]}
          ShowMaskFx {type=MovieSceneScriptingFloatChannel, keys [[0,0],[0.1237,1],[1.0123,1],[1.1137,0]]}
          GlowColorR {type=MovieSceneScriptingFloatChannel, const [1,5]}
          GlowColorG {type=MovieSceneScriptingFloatChannel, thinned_from=5, keys [[0,0.4977],[18.7557,0.4977],[19.402,1],[20,0.4977]]}
          GlowColorB {type=MovieSceneScriptingFloatChannel, thinned_from=5, keys [[0,0],[18.7557,0],[19.402,1],[20,0]]}
          GlowColorA {type=MovieSceneScriptingFloatChannel, const [1,4]}
         ]
        }]
      }]
    }]}
 }
]
graphs [
 EventGraph {nodes [
   Event Construct()
   > n1 = Cast To FortPlayerController(Object=GetOwningPlayer())
   > Set As Fort Player Controller(As Fort Player Controller:Fort Player Controller Object Reference=@n1) -> (Output_Get:Fort Player Controller Object Reference)
   > n2 = Sequence()
   n3 = Get As Fort Player Controller() -> (As Fort Player Controller:Fort Player Controller Object Reference)
   > n4 = WaitGameplayTagAddToActor!(
    execute:Exec
    TargetActor:Actor Object Reference=@n3
    Tag:Gameplay Tag Structure=(TagName="Granted.Cooldown.Masamune")
    OnlyTriggerOnce:bool=false
   ) -> (then:Exec, AsyncTaskProxy:Ability Async Wait Gameplay Tag Added Object Reference, Added:Exec) {is=title K2Node_AsyncAction}
   n5 = GetActiveEffectsWithAllTags!(
    execute=@n4.Added
    self:Ability System Component Object Reference=GetAbilitySystemComponent!(Actor:Actor Object Reference=@n3) -> (ReturnValue:Ability System Component Object Reference)
    Tags:Gameplay Tag Container Structure=(GameplayTags=((TagName="Granted.Cooldown.Masamune")))
   ) -> (then, ReturnValue:Array of Active Gameplay Effect Handle Structures)
   > Branch(Condition="integer > integer"(A=Length(TargetArray:Array of Active Gameplay Effect Handle Structures=@n5)))
   n6 = "Get (a copy)"(Array:Array of Active Gameplay Effect Handle Structures=@n5) -> (Output:Active Gameplay Effect Handle Structure)
   > n7 = GetActiveGameplayEffectRemainingDuration!(execute, ActiveHandle:Active Gameplay Effect Handle Structure=@n6) -> (then, ReturnValue:float)
   > Set CurrentCooldown(CurrentCooldown:double=@n7) -> (Output_Get:double)
   > n8 = GetActiveGameplayEffectTotalDuration!(execute, ActiveHandle:Active Gameplay Effect Handle Structure=@n6) -> (then, ReturnValue:float)
   > HanldeCooldown!(execute, self, CurrentCooldown:double=Get CurrentCooldown() -> (CurrentCooldown:double), TotalCooldown:double=@n8) -> (then)
   n9 = WaitGameplayTagRemoveFromActor!(
    execute:Exec=@n2.then_1
    TargetActor:Actor Object Reference=Get As Fort Player Controller()
    Tag:Gameplay Tag Structure
    OnlyTriggerOnce:bool=false
   ) -> (then:Exec, AsyncTaskProxy:Ability Async Wait Gameplay Tag Removed Object Reference, Removed:Exec) {is=title K2Node_AsyncAction}
   n10 = GetDynamicMaterial(execute=@n9.Removed, self=Get Reticle() -> (Reticle:Fort SZAware Image Object Reference))
   > SetScalarParameterValue(self=@n10, ParameterName=AnimateProgress, Value=1)
  ]}
 HanldeCooldown {nodes [
   HanldeCooldown!() -> (then, CurrentCooldown:double=~, TotalCooldown:double=~) {is=title K2Node_FunctionEntry}
   n1 = Get TotalCooldown() -> (TotalCooldown:double)
   n2 = Get CurrentCooldown()
   > PlayAnimation(
    InAnimation=Get Anim_CoolDown_02() -> (Anim_CoolDown_02:Widget Animation Object Reference)
    StartAtTime=float / float(A=float - float(A=@n1, B=@n2), B=Get TotalCooldown())
    PlaybackSpeed=float / float(A=1, B=Get TotalCooldown())
   )
  ]}
]
variables [
 As Fort Player Controller {type=FortPlayerController, default=null}
 Charging Component {type=B_MasamuneChargeComponent_C, default=null}
 CurrentCooldown {type=number}
]
widgets {
 info {
  inheritedWidgetCount=1
  parentClass=/Script/FortniteUI.FortWeaponReticleExtensionWidgetBase
  rootWidgetClass=&Overlay
  widgetCount=4
 }
 widgets [
  Overlay_0 {widgetClassPath=&Overlay, widget=Overlay_0}
  Image_ProgressBar {
   widgetClassPath=&Image
   bIsVariable=true
   parent=Overlay_0
   widget=Image_ProgressBar
   properties {
    visibility=Hidden
    brush {
     bIsDynamicallyLoaded=false
     drawAs=Image
     imageType=FullColor
     mirroring=NoMirror
     resourceName=None
     resourceObject=/MasamuneGameplay/Assets/UI/Materials/M_UI_MasamuneCharge
     tiling=NoTile
     imageSize=<250 250>
     margin=<ltrb 0 0 0 0>
     outlineSettings {
      bUseBrushTransparency=false
      roundingType=HalfHeightRadius
      width=0
      color {colorUseRule=UseColor_Specified, specifiedColor=<rgba 0 0 0 0>}
      cornerRadii=<0 0 0 0>
     }
     tintColor {colorUseRule=UseColor_Specified, specifiedColor=<rgba 1 1 1 1>}
     uVRegion {bIsValid=false, max=<0 0>, min=<0 0>}
    }
   }
   slot {class=OverlaySlot}
  }
  Reticle {
   widgetClassPath=/Script/FortniteUI.FortSZAwareImage
   bIsVariable=true
   parent=Overlay_0
   widget=Reticle
   properties {brush {
     bIsDynamicallyLoaded=false
     drawAs=Image
     imageType=FullColor
     mirroring=NoMirror
     resourceName=None
     resourceObject=/MasamuneGameplay/Assets/UI/Materials/M_UI_Masamune_Reticle
     tiling=NoTile
     imageSize=<132 132>
     margin=<ltrb 0 0 0 0>
     outlineSettings {
      bUseBrushTransparency=false
      roundingType=HalfHeightRadius
      width=0
      color {colorUseRule=UseColor_Specified, specifiedColor=<rgba 0 0 0 0>}
      cornerRadii=<0 0 0 0>
     }
     tintColor {colorUseRule=UseColor_Specified, specifiedColor=<rgba 1 1 1 1>}
     uVRegion {bIsValid=false, max=<0 0>, min=<0 0>}
    }}
   slot {class=OverlaySlot, properties {horizontalAlignment=HAlign_Center, verticalAlignment=VAlign_Center}}
  }
  TutorialHighlight {widgetClassPath=&TutorialHighlight}
 ]
}
